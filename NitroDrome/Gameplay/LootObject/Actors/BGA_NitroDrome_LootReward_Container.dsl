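; Capturable NitroDrome loot crate objective that starts a countdown on interaction, then tosses out reward loot and ammo.
; A replicated state enum drives server logic and client cosmetics from spawning through active to completed, with hotfixable timings and loot tiers.
; Timelines feed custom primitive data on the crate material to show progress and takeover, restored for clients entering relevancy.
; It also handles squad assignment relevancy, match-end analytics and capture reset.
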(event Custom|UpdateTakeoverStateForClientsEnteringRelevancy
  (params (IsCaptured :type Boolean))
  (bind _returnvalue (Math|Float|SelectFloat 1.0 0.0 IsCaptured))
  (bind _interactionobject (Variables|Default|GetInteractionObject))
  ;[0.2824 0.3886 0.4149 1 120] INITIALIZATION
  (Rendering|Material|SetCustomPrimitiveDataFloat _interactionobject 2 _returnvalue)
  (Rendering|Material|SetCustomPrimitiveDataFloat _interactionobject 1 _returnvalue)
  (Rendering|Material|SetCustomPrimitiveDataFloat _interactionobject 0 (Math|Float|SelectFloat 3.0 0.0 IsCaptured)))

(event Custom|UpdateMMOCrateMaterialCaptureState
  (params (ColorValue :type Float%20%28double-precision%29))
  ;[0.2824 0.3886 0.4149 1 120] INITIALIZATION
  ;;[0.2058 0 1] [CLIENT MULTICAST] Update MMO crate material to adjust color of the progress bar based on the capture state
  (Rendering|Material|SetCustomPrimitiveDataFloat (Variables|Default|GetInteractionObject) 0 ColorValue))

(event EventBeginPlay
  ;[0.2824 0.3886 0.4149 1 120] INITIALIZATION
  ;;[0 1 0.03889] EventBeginPlay
  ;; Set up the initial state of the objective when created. Should be set "inactive" (no visibility, interaction, or collision)
  (|CallParentFunction)
  ;;;[0.1435 0.221 1] AUTHORITY
  (Utilities|FlowControl|Sequence
    (:then_0
      (Utilities|FlowControl|Delay 1.0)
      (bind _returnvalue (Actor|GetActorOfClass "/NitroDrome/Gameplay/Actors/NitroDromeGameplayVolume.NitroDromeGameplayVolume_C"))
      (Variables|Default|SetNitroDromeGameplayVolume _returnvalue)
      (bind _output_get (Variables|Default|SetCaptureTime (ScalableFloat|GetValueatLevel (Variables|Objective|GetSF_ObjectiveCountdownDuration))))
      (Variables|Default|SetAdjustedCaptureTime _output_get)
      (Default|SetMMOMiniMapIcon false))
    (:then_1
      (Networking|SwitchHasAuthority
        (:Authority
          ;;[0 0.08472 1] Enabled / Disabled HF
          (if (not (ScalableFloat|GetValueasBool (Variables|Default|GetHF_Enabled)))
            (Actor|DestroyActor)))))
    (:then_2
      (Networking|SwitchHasAuthority
        (:Authority
          (CallFunction|UpdateTakeoverStateforClientsEnteringRelevancy (Variables|Default|GetIsMMOCaptureComplete)))
        (:Remote
          ;;;;[0 0 0] If Objective is already Active display Visual Timer
          (if (Utilities|Enum|Equal(Enum) (Variables|Objective|States|GetECurrentObjectiveState_RepNotify) "NewEnumerator3")
            (CallFunction|CosmeticsStateActiveObjectiveStartBurst)))))
    (:then_3
      ;;;;[0 1 0.03889] Sequence [0] Set the state of the objective to "Spawning"
      (Variables|Objective|States|SetECurrentObjectiveState_RepNotify "NewEnumerator1"))))

(event Custom|UpdateTakeoverVisuals
  ;[0.2824 0.3886 0.4149 1 120] INITIALIZATION
  (Utilities|FlowControl|Sequence
    (:then_0
      ;;[0 0.09804 1] Set play rate for timeline hotfixable
      (Components|Timeline|SetPlayRate (Variables|BGA_NitroDrome_LootReward_Container|GetUpdateTakeoverVisualsTimeline) (Math|Float|SafeDivide 1.0 (ScalableFloat|GetValueatLevel (Variables|Default|GetSF_Takeover_Time))))
      (Timeline|UpdateTakeoverVisualsTimeline PlayFromStart
        (:Finished)))
    (:then_1
      (Components|Timeline|SetPlayRate (Variables|BGA_NitroDrome_LootReward_Container|GetUpdate%20MMO%20Crate%20Takeover%20Visuals%20Timeline) (Math|Float|SafeDivide 1.0 (ScalableFloat|GetValueatLevel (Variables|Default|GetSF_MMOCrate_Takeover_Time))))
      (bind (_direction _newtrack_0) (Timeline|Update%20MMO%20Crate%20Takeover%20Visuals%20Timeline PlayFromStart)
        (:Update
          (Rendering|Material|SetCustomPrimitiveDataFloat (Variables|Default|GetInteractionObject) 2 _newtrack_0))))))

(event Custom|RemoveTakeoverVisuals
  ;[0.2824 0.3886 0.4149 1 120] INITIALIZATION
  (bind (_direction _newtrack_0) (Timeline|Reset%20Takeover%20Visuals%20Timeline PlayFromStart)
    (:Update
      (Rendering|Material|SetCustomPrimitiveDataFloat (Variables|Default|GetInteractionObject) 2 _newtrack_0))))

(event Custom|OnLocalViewTargetChanged
  (params (PlayerController :type Fort%20Player%20Controller%20Object%20Reference) (OldViewTarget :type Actor%20Object%20Reference) (NewViewTarget :type Actor%20Object%20Reference))
  ;[0.2824 0.3886 0.4149 1 120] INITIALIZATION
  ;;[0 1 0.03889] EventBeginPlay
  ;; Set up the initial state of the objective when created. Should be set "inactive" (no visibility, interaction, or collision)
  ;;;[0.1922 1 0.9216] OnLocalViewTargetChanged - Call ObjectiveStateChanged to refresh the visuals for the new view target
  (CallFunction|ObjectiveStateChanged))

(event Custom|Adjust%20MMO%20Crate%20Progress
  (params (Progress :type Float%20%28double-precision%29))
  ;[0.2824 0.3886 0.4149 1 120] INITIALIZATION
  ;;[0.2058 0 1] [CLIENT MULTICAST] adjust progress bar in MMO crate material
  (Rendering|Material|SetCustomPrimitiveDataFloat (Variables|Default|GetInteractionObject) 1 Progress))

(event Custom|InitAssignedObjectiveOnCreation
  (params (AsignedPlayers :type Array%20of%20Fort%20Player%20State%20Athena%20Object%20References) (AssignedSquadMidpoints :type Array%20of%20Vectors))
  ;[0.2824 0.3886 0.4149 1 120] INITIALIZATION
  ;;[0 1 0.03889] InitAssignedObjectiveParameters
  ;; Initial setup for the object with variables passed from the objective spawner
  ;;[Called from the Spawning in the Assigned Objective Game State]
  (Assignment|AddPlayerstoAssignedObjective AsignedPlayers)
  ;;;[0.2435 0.1798 0.4062] Setup initial variables for analytics
  (Analytics|AnalyticsGetonInitAttributes AssignedSquadMidpoints))

(event Custom|LowerCaptureTimer
  ;[0.2824 0.3886 0.4149 1 120] INITIALIZATION
  (Variables|Default|SetCaptureTime 5.0))

(event Custom|ObjectiveState_Changed
  ;[0.2824 0.3886 0.4149 1 120] INITIALIZATION
  ;;[0.2687 0.7669 1] ObjectiveState_Changed
  ;; Runs when the enum objective state is changed to determine which state logic should be run
  (if (Utilities|Enum|NotEqual(Enum) (Variables|Objective|States|GetECurrentObjectiveState_RepNotify) (Variables|Objective|States|GetELastObjectiveStateProcessed_Replicated))
    (ObjectiveStates|SwitchonStateChangeDetermineStart) ; selects the objective state "Start" event to call \nbased on the current objective state
    (else
      _))
  (Networking|SwitchHasAuthority
    (:Authority
      ;;;[0.2687 0.7669 1] set last state after a short delay to distinguish between new state changes and relevancy on the client
      (Utilities|FlowControl|Delay 0.5)
      (Variables|Objective|States|SetELastObjectiveStateProcessed_Replicated (Variables|Objective|States|GetECurrentObjectiveState_RepNotify)))
    (:Remote
      (ObjectiveStates|SwitchonStateChangeDetermineCosmeticPersist) ; selects the objective state "Persist" event to call \nbased on the current objective state 
    )))

(event Custom|CosmeticsState_Spawning_StartBurst)

(event Custom|CosmeticsState_Spawning_PersistentLoop)

(event Custom|ObjectiveState_Spawning
  ;[0.2824 0.3886 0.4149 1 120] Objective State: SPAWNING
  ;;[0.03332 0.3212 0.2859] ObjectiveState_Spawning_Start
  (Variables|Objective|States|SetbObjectiveSpawnedAndReady true)
  (Variables|Objective|States|SetECurrentObjectiveState_RepNotify "NewEnumerator2"))

(event Custom|CosmeticsState_Spawning_EndBurst)

(event Custom|ObjectiveState_Spawning_End
  ;[0.2824 0.3886 0.4149 1 120] Objective State: SPAWNING
  ;;[0.03332 0.3212 0.2859] ObjectiveState_Spawning_End
  ;; Set visuals and collision/interaction; specific end call for spawning related since most actions for the spawning state happen at the end
  (Networking|SwitchHasAuthority
    (:Authority
      (Variables|Objective|States|SetbObjectiveSpawnedAndReady true) ; using a seperate bool for the Spawned and Ready state \nso that it will always be set instead of requiring it to be\nset on all the following states in the sequence
      ;;;[0.2435 0.1798 0.4062] Analytics: Set spawned time for analytics
      (Variables|Analytics|SetAnalyticsHelper_ObjectiveSpawnedTime (Fortnite|Sync|Time|GetServerWorldTimeSeconds)))))

(event Custom|ObjectiveState_AvailableIdle
  ;[0.2824 0.3886 0.4149 1 120] Objective State: AVAILABLE IDLE - waiting for interact, broadcasting location to assigned teams
  ;;[0.03332 0.3212 0.2859] ObjectiveState_AvailableIdle
  ;; Send quest update that the objective is spawned and ready  - DecayingLoot child BP will make more use of this state
  (Networking|SwitchHasAuthority
    (:Authority
      (Quests|SendTutorialQuestEventtoAssignedPlayers))))

(event Custom|CosmeticsState_AvailableIdle_StartBurst)

(event Custom|CosmeticsState_AvailableIdle_PersistentLoop)

(event Custom|MMOTimerToSpawnLoot
  ;[0.2824 0.3886 0.4149 1 120] Objective State: ACTIVE OBJECTIVE - players are in active contest, objective timer is started
  ;;[0.2058 0 1] Finish MMO capture, spawn loot, set takeover visuals
  (CallFunction|UpdateMMOCrateMaterialCaptureState 3.0)
  (CallFunction|UpdateTakeoverVisuals)
  (Variables|Default|SetIsMMOCaptureComplete true)
  ;;;[0 1 0.04981] Start Takeover GCNL
  (CallFunction|StartTakeOverGCNL)
  (Default|SetMMOMiniMapIcon true))

(event Custom|InstantCapture
  ;[0.2824 0.3886 0.4149 1 120] Objective State: ACTIVE OBJECTIVE - players are in active contest, objective timer is started
  ;;[0.2058 0 1] Finish MMO capture, spawn loot, set takeover visuals
  (Variables|Objective|States|SetbObjectiveSpawnedAndReady false)
  (CallFunction|SpawnRewardItems) ; timer can't be zero or less, so spawn immediately
  (CallFunction|UpdateMMOCrateMaterialCaptureState 3.0)
  (CallFunction|UpdateTakeoverVisuals)
  (Variables|Default|SetIsMMOCaptureComplete true)
  ;;;[0 1 0.04981] Start Takeover GCNL
  (CallFunction|StartTakeOverGCNL)
  (Default|SetMMOMiniMapIcon true))

(event Interaction|EventBlueprintOnInteract
  (params (InteractingPawn :type Fort%20Pawn%20Object%20Reference) (InteractionBeingAttempted :type EInteractionBeingAttempted%20Enum :default "FirstInteraction"))
  ;[0.2824 0.3886 0.4149 1 120] Objective State: ACTIVE OBJECTIVE - players are in active contest, objective timer is started
  ;;[0 0.3455 0.006439] EventBlueprintOnInteract [SERVER]
  ;; Verify the interaction is valid, set current control owner, and call the objective start + replicate to enable client FX
  ;;;[0.3202 0.7376 1] Validate the interaction against assigned players
  (Utilities|IsValid InteractingPawn
    (:Is%20Valid
      (Variables|Objective|States|SetbObjectiveSpawnedAndReady false)
      (GameplayCue|ExecuteGameplayCue self (Variables|Default|GetOn%20Objective%20Started%20Gameplay%20Cue%20Tag))
      (Variables|Objective|States|SetECurrentObjectiveState_RepNotify "NewEnumerator3")
      (CallFunction|UpdateMMOCrateMaterialCaptureState 2.0)
      (CallFunction|SetandStartObjectiveTimer (Variables|Default|GetCaptureTime)))))

(event Custom|StartObjectiveTimerEvent
  ;[0.2824 0.3886 0.4149 1 120] Objective State: ACTIVE OBJECTIVE - players are in active contest, objective timer is started
  ;;[0 0.3455 0.006439] EventBlueprintOnInteract [SERVER]
  ;; Verify the interaction is valid, set current control owner, and call the objective start + replicate to enable client FX
  (Variables|Objective|States|SetbObjectiveSpawnedAndReady false)
  (GameplayCue|ExecuteGameplayCue self (Variables|Default|GetOn%20Objective%20Started%20Gameplay%20Cue%20Tag))
  (Variables|Objective|States|SetECurrentObjectiveState_RepNotify "NewEnumerator3")
  (CallFunction|UpdateMMOCrateMaterialCaptureState 2.0)
  (CallFunction|SetandStartObjectiveTimer (Variables|Default|GetCaptureTime)))

(event Custom|ResetCapture
  ;[0.2824 0.3886 0.4149 1 120] Objective State: ACTIVE OBJECTIVE - players are in active contest, objective timer is started
  ;;[0.4307 0 1] Reset MMO state
  (Variables|Objective|States|SetbObjectiveSpawnedAndReady true)
  (Variables|Default|SetIsMMOCaptureComplete false)
  (CallFunction|UpdateMMOCrateMaterialCaptureState 0.0)
  (CallFunction|RemoveTakeoverVisuals)
  (CallFunction|AdjustMMOCrateProgress 0.0)
  (Default|SetMMOMiniMapIcon false)
  (Variables|Objective|States|SetECurrentObjectiveState_RepNotify "NewEnumerator2")
  (GameplayCue|ExecuteGameplayCue self (Variables|Default|GetReplenish%20MMO%20Coitainer%20Gameplay%20Cue%20Tag))
  (Networking|SwitchHasAuthority))

(event Custom|CosmeticsState_ActiveObjective_StartBurst
  ;[0.04953 0 0.1354 1 120] CLIENT COSMETIC EVENTS (VFX/SFX)
  ;;[0.04953 0 0.1354 1 50] COSMETICS: ActiveObjective
  ;;;[0.3658 0 1] Overrides available in the child BPs
  (Networking|CanExecuteCosmeticEvents
    (:True
      (GameplayEventRouter|BroadcastEventBackwardCompatible (GameplayEventRouter|GetGlobalEventRouter) "(TagName=\"EventMessage.Athena.NitroDrome.ObjectiveActivated\")" (Utilities|Struct|MakeNitroDromeObjectiveActivated "Neutral" (Variables|Default|GetActivationTime))))))

(event Custom|CosmeticsState_ActiveObjective_PersistentLoop
  ;[0.04953 0 0.1354 1 120] CLIENT COSMETIC EVENTS (VFX/SFX)
  ;;[0.04953 0 0.1354 1 50] COSMETICS: ActiveObjective
  ;;;[0.3658 0 1] Overrides available in the child BPs
  (Audio|Components|Audio|FadeIn (Variables|Default|GetSFX_ObjectiveTimerActive_loop) 0.5 1.0 0.0 "Linear" :!class /Script/Engine.AudioComponent))

(event Custom|ObjectiveState_ActiveObjective
  ;[0.2824 0.3886 0.4149 1 120] Objective State: ACTIVE OBJECTIVE - players are in active contest, objective timer is started
  ;;[0.03332 0.3212 0.2859] ObjectiveState_ActiveObjective
  ;; Player has interacted with the objective and the primary contest is begun, make sure that the objective is awake and start the timer
  (Networking|SwitchHasAuthority
    (:Authority
      (Variables|Default|SetActivationTime (Fortnite|Sync|Time|GetServerWorldTimeSeconds))
      ;;;[0.2153 0.3363 1] Make sure we're awake
      (Networking|SetNetDormancy "DORM_Awake")
      (Networking|ForceNetUpdate)
      (CallFunction|SetandStartObjectiveTimer (ScalableFloat|GetValueatLevel (Variables|Objective|GetSF_ObjectiveCountdownDuration)))
      ;;;[0.2435 0.1798 0.4062] Analytics: Calculate elapsed time since the objective was spawned until activated
      (Variables|Analytics|SetAnalyticsAttribute_ElapsedTimeToObjectiveActivation (- (Fortnite|Sync|Time|GetServerWorldTimeSeconds) (Variables|Analytics|GetAnalyticsHelper_ObjectiveSpawnedTime))))))

(event Custom|SetAndStartObjectiveTimer
  (params (InputCaptureTime :type Float%20%28double-precision%29))
  ;[0.2824 0.3886 0.4149 1 120] Objective State: ACTIVE OBJECTIVE - players are in active contest, objective timer is started
  ;[1 0.9522 0.04365] TODO: cleanup and consolidate this
  (bind _returnvalue (Utilities|Time|SetTimerbyEvent (EventDispatchers|CreateEvent :!event Custom|ResolveObjectiveTimerEnd) InputCaptureTime))
  (Variables|Objective|SetTH_ObjectiveTimer _returnvalue)
  (bind _returnvalue_2 (Utilities|Time|SetTimerbyEvent (EventDispatchers|CreateEvent :!event Custom|MMOTimerToSpawnLoot) InputCaptureTime))
  (Variables|Default|SetObjectiveCountdownTimerHandle _returnvalue_2)
  ;;[0 0.09804 1] Set play rate for timeline hotfixable
  (Components|Timeline|SetPlayRate (Variables|BGA_NitroDrome_LootReward_Container|GetMMO%20Container%20Fill%20Progress) (Math|Float|SafeDivide 1.0 InputCaptureTime))
  (bind (_direction _newtrack_2) (Timeline|MMO%20Container%20Fill%20Progress PlayFromStart)
    (:Update
      (CallFunction|AdjustMMOCrateProgress _newtrack_2))))

(event Custom|ResolveObjectiveTimerEnd
  ;[0.2824 0.3886 0.4149 1 120] Objective State: COMPLETED IDLE - spawn the rewards and disable most interactivity on the objective, cleanup
  ;;[0.3202 0.7376 1] ResolveObjectiveTimerEnd
  ;; On objective timer completed, reward drop and begins resolving the objective (set variables, cleaup, visuals, etc.)
  ;;[Called from objective timer ending]
  (Networking|SwitchHasAuthority
    (:Authority
      (Variables|Objective|States|SetECurrentObjectiveState_RepNotify "NewEnumerator4"))))

(event Custom|CosmeticsState_CompletedIdle_StartBurst)

(event Custom|CosmeticsState_CompletedIdle_PersistentLoop
  ;[0.04953 0 0.1354 1 120] CLIENT COSMETIC EVENTS (VFX/SFX)
  ;;[0.04953 0 0.1354 1 50] COSMETICS: CompletedIdle
  ;;;[0.3658 0 1] Overrides available in the child BPs
  (Audio|Components|Audio|FadeOut (Variables|Default|GetSFX_ObjectiveTimerActive_loop) 0.4 0.0 "Linear" :!class /Script/Engine.AudioComponent))

(event Custom|ObjectiveState_CompletedIdle
  (bind _rewardspawndelay (Variables|Objective|Reward|GetRewardSpawnDelay))
  ;[0.2824 0.3886 0.4149 1 120] Objective State: COMPLETED IDLE - spawn the rewards and disable most interactivity on the objective, cleanup
  ;;[0.03332 0.3212 0.2859] ObjectiveState_CompletedIdle
  ;; Spawn the reward and disable most interactivity
  (Networking|SwitchHasAuthority
    (:Authority
      (Utilities|FlowControl|Sequence
        (:then_0
          ;;;[0 0.04199 1] Sequence [0]: Spawn the reward - delay the reward based on the delay time - set by the child container opening duration
          (if (> _rewardspawndelay 0)
            (bind _returnvalue (Utilities|Time|SetTimerbyEvent (EventDispatchers|CreateEvent :!event Custom|SpawnRewardItems) _rewardspawndelay))
            (Variables|Objective|Reward|SetTH_SpawnRewardDelay _returnvalue)
            (else
              (CallFunction|SpawnRewardItems) ; timer can't be zero or less, so spawn immediately
              (CallFunction|UpdateMMOCrateMaterialCaptureState 3.0)
              (CallFunction|UpdateTakeoverVisuals))))
        (:then_1
          ;;;[0.3202 0.7376 1] Sequence [1]: Set interact and minimap to finished state
          (Variables|Objective|Interaction|SetbCanInteract false))
        (:then_2)))))

(event EventOnDamageServer
  (params (Damage :type Float%20%28single-precision%29) (DamageTags :type Gameplay%20Tag%20Container%20Structure%20%28by%20ref%29) (Momentum :type Vector :default "0, 0, 0") (HitInfo :type Hit%20Result%20Structure%20%28by%20ref%29) (InstigatedBy :type Controller%20Object%20Reference) (DamageCauser :type Actor%20Object%20Reference) (EffectContext :type Gameplay%20Effect%20Context%20Handle%20Structure))
  ;[0.2824 0.3886 0.4149 1 120] Objective State: COMPLETED IDLE - spawn the rewards and disable most interactivity on the objective, cleanup
  ;;[0.3837 0.2485 0.09659] EventOnDamageServer
  ;; After being set dormant, objects that are damaged do not update their HP bar- set dormancy active so feedback can be shown
  ;; (Only expected if HF_bDamagableAfterClaimed sets the objectitve damagable in ObjectiveState_CompletedIdle)
  (Networking|SetNetDormancy "DORM_Awake")
  (Networking|ForceNetUpdate))

(event Custom|SpawnRewardItems
  (bind (_rewardloottier _foundtier) (Default|GetCurrentLootTier))
  ;[0.2824 0.3886 0.4149 1 120] Objective State: COMPLETED IDLE - spawn the rewards and disable most interactivity on the objective, cleanup
  ;;[0 0.04199 1] SpawnRewardItems
  ;; Spawn the reward tier items
  (bind _itempickupsspawned (Class|AthenaFunctionLibrary|SpawnLootfromTableUsingItemEntry (select _foundtier _rewardloottier (Variables|Objective|Reward|GetDefaultRewardLootTierGroup)) 0 -1 (+ (Transformation|GetActorLocation) (+ (* (Transformation|GetActorUpVector) 1e+02) (* (Transformation|GetActorRightVector) 2.5e+02))) (* (Transformation|GetActorRightVector) 2.5e+02) 0 true true false 0 "Tossed" "SupplyDrop"))
  (bind _output_get (Variables|Default|SetReturn_SpawnedLoot _itempickupsspawned))
  (for _array_element _output_get
    (Utilities|Casting|CastToFortWeaponItemDefinition (Class|FortItemFunctionLibrary|BreakFortItemEntry (Pickup|GetItemEntry _array_element 0)))
    (bind _returnvalue (Fortnite|Weapon|CreateTossAmmoPickupforWeaponItemDefinitionatLocation 0 (Variables|Default|GetRewardAmmoSourceTag) (Transformation|GetActorLocation)))
    (Utilities|IsValid _returnvalue
      (:Is%20Valid
        (Utilities|Array|Add (Variables|Default|GetReturn_SpawnedLoot) _returnvalue) ; add ammo to spawned items
      )))
  (CallFunction|TossPickupstoEnsureLootStickstoTrain (Variables|Default|GetReturn_SpawnedLoot)))

(event Custom|DestroyAfterRoundEnd
  (params (InstigatorPC :type Fort%20Player%20Controller%20Object%20Reference) (EndMethod :type EFortMinigameEnd%20Enum :default "AbandonGame") (NextState :type EFortMinigameState%20Enum :default "PreGame"))
  ;;[1 0.3479 0] DestroyAfterRoundEnd
  ;; Only Called On Creative Royale Playlist (bound in BeginPlay)
  ;[0.2824 0.3886 0.4149 1 120] CREATIVE - Creative Royale
  (Actor|SetLifeSpan (Variables|Creative|GetObjectiveResolvedLifespan_RoundEnd)))

(event EventEndPlay
  (params (EndPlayReason :type EEndPlayReason%20Enum :default "Destroyed"))
  (bind _self self)
  (bind _returnvalue (Minigame|GetMinigame _self :!class /Script/FortniteGame.MinigameBlueprintLibrary))
  (bind _squadswearealwaysrelevantto (Variables|Objective|Assigned|GetSquadsWeAreAlwaysRelevantTo))
  ;;[1 0.3479 0] EndPlay
  ;; Cleanup timers, unbind events, and remove special squad relevancy
  ;[0.2824 0.3886 0.4149 1 120] END PLAY - cleanup and destruction
  (|CallParentFunction EndPlayReason)
  (Utilities|FlowControl|Sequence
    (:then_0
      ;;;[1 0.3479 0] [0] Remove special squad relevancy
      (for _array_element _squadswearealwaysrelevantto
        (Class|FortPlayerStateAthena|OnSquadSpecificActorDestroyed _array_element self))
      (Utilities|Array|Clear _squadswearealwaysrelevantto))
    (:then_1
      ;;;[1 0.3479 0] [1] Clear and Invalidate timers
      (Utilities|Time|ClearandInvalidateTimerbyHandle (Variables|Objective|GetTH_ObjectiveTimer))
      (Utilities|Time|ClearandInvalidateTimerbyHandle (Variables|Objective|Reward|GetTH_SpawnRewardDelay)))
    (:then_2
      ;;;[1 0.3479 0] [2] Unbind events
      (Utilities|IsValid _returnvalue
        (:Is%20Valid
          (Minigame|UnbindEventfromMinigameEndGameDelegate _returnvalue (EventDispatchers|CreateEvent :!event Custom|DestroyAfterRoundEnd)))
        (:Is%20Not%20Valid
          _))
      (bind _gamephaseupdatedasynctask (Variables|Default|GetGamePhaseUpdatedAsyncTask)
        (:then
          (GameplayEventRouter|StopListeningtoStatefulEvent _gamephaseupdatedasynctask))
        (:else)))
    (:then_3
      ;;;[1 0.3479 0] [3] Unbind OnViewTargetChanged
      (Networking|SwitchHasAuthority
        (:Remote
          (for _array_element_2 (Fortnite|GetLocalFortPlayerControllers true true)
            (Utilities|IsValid _array_element_2
              (:Is%20Valid
                (EventDispatchers|UnbindEventfromOnViewTargetChanged _array_element_2 (EventDispatchers|CreateEvent :!event Custom|OnLocalViewTargetChanged))))))))))

(fn AddPlayersToAssignedObjective
  (params (PlayersToAdd :type Array%20of%20Fort%20Player%20State%20Athena%20Object%20References))
  (bind _squadswearealwaysrelevantto (Variables|Objective|Assigned|GetSquadsWeAreAlwaysRelevantTo))
  ;[0 0.04199 1] AddPlayersToAssignedObjective
  ; Store the assigned players and set relevancy so only the assigned teams keep relevancy
  (Networking|SwitchHasAuthority
    (:Authority
      ;;[0 0.04199 1] Store newly assigned players and replicate to clients
      (Utilities|Array|AppendArray (Variables|Objective|Assigned|GetPlayersAssignedToObjective) PlayersToAdd)
      ;;[0.04501 1 0.5772] For Each newly added player.
      (for _array_element PlayersToAdd
        (Utilities|IsValid _array_element
          (:Is%20Valid
            (Utilities|FlowControl|Sequence
              (:then_0
                ;;;[0.5552 0 1] Have the marker icon show up as discovered for all players assigned to this objective.
                (Utilities|Casting|CastToFortPlayerController (PlayerState|GetPlayerController _array_element)))
              (:then_1
                ;;;[0.4996 0.9801 1] Make this objective always relevant for members of the squad assigned to it.
                ;;; This relies on the assumption that if this objective was assigned to a player it is assigned to all players within that squad.
                (if (not (Utilities|Array|ContainsItem _squadswearealwaysrelevantto (Class|FortPlayerStateAthena|GetSquadId _array_element)))
                  (Utilities|Array|Add _squadswearealwaysrelevantto (Class|FortPlayerStateAthena|GetSquadId _array_element))
                  (Class|FortPlayerStateAthena|OnSquadSpecificActorSpawned (Class|FortPlayerStateAthena|GetSquadId _array_element) self))))))))))

(fn Analytics_GetOnInitAttributes
  (params (AssignedSquadAvgLocations :type Array%20of%20Vectors))
  ;[0.2435 0.1798 0.4062] Analytics_GetOnInitAttributes
  ; Set initial variables for the analytics send that won't change or need to be stored now
  (Utilities|FlowControl|Sequence
    (:then_0
      ;;[0.2435 0.1798 0.4062] Calculate the average distance of each squad to the objective
      (for _array_element AssignedSquadAvgLocations
        (Utilities|Array|Add (Variables|Analytics|GetAnalyticsAttribute_AssignedTeamDistance) (Math|Vector|Distance2D(Vector) _array_element (Transformation|GetActorLocation)))))
    (:then_1
      ;;[0.2435 0.1798 0.4062] Storing this now for if teams leave the match
      (Variables|Analytics|SetAnalyticsAttribute_AssignedAsContested (> (Utilities|Array|Length (Assignment|GetAllAssignedSquadIDs)) 1)))))

(fn BlueprintCanInteract
  (params (InteractingPawn :type Fort%20Pawn%20Object%20Reference) (InteractionBeingAttempted :type EInteractionBeingAttempted%20Enum :default "FirstInteraction") (InteractionType :type TInteractionType%20Enum :default "IT_NoInteraction"))
  (return (Variables|Objective|Interaction|GetbCanInteract)))

(fn BlueprintGetInteractionString
  (params (InteractingPawn :type Fort%20Pawn%20Object%20Reference) (InteractionBeingAttempted :type EInteractionBeingAttempted%20Enum :default "FirstInteraction"))
  (return (Variables|Default|GetInteractText)))

(fn BlueprintGetInteractionTime
  (params (InteractingPawn :type Fort%20Pawn%20Object%20Reference) (InteractionBeingAttempted :type EInteractionBeingAttempted%20Enum :default "FirstInteraction"))
  ;[0.3202 0.7376 1] BlueprintGetInteractionTime
  ; Players assigned to the objective should have an interaction time, but unassigned players should immediate get the failed feedback from EventBlueprintOnInteract
  ;[Override Parent]
  ;;[1 0.3594 0.9217] Assigned players use set hotfixable value
  ;; Unassigned player should immediately get feeback (0) or use their interaction time
  (return (select (Assignment|IsPawnAssignedtoObjective InteractingPawn) (ScalableFloat|GetValueatLevel (Variables|Objective|Interaction|GetSF_InteractionTime)) (select (ScalableFloat|GetValueasBool (Variables|Objective|Interaction|GetSF_AllowUnassignedInteraction)) (ScalableFloat|GetValueatLevel (Variables|Objective|Interaction|GetSF_UnassignedInteractTime)) 0.0)) true))

(fn BuildDataRegistryResolverScope
  (bind _self self)
  (bind _returnvalue (Minigame|GetMinigame _self :!class /Script/FortniteGame.MinigameBlueprintLibrary))
  ;[1 0.8475 0.2776] Gives the name of the volume as a scope to the data registry. Needed for calls to GetValueWithContext to work
  (Utilities|IsValid _returnvalue
    (:Is%20Valid
      (return (Utilities|Array|MakeArray (SpatialGameplay|GetSpatialGameplayInterfaceFNameforActor _returnvalue)) 0 true))
    (:Is%20Not%20Valid
      (return 0 0 false))))

(fn GetPlaylistSquadSizeTag
  (bind _returnvalue (Fortnite|Athena|GameState|GetGameStateAthena))
  ;[1 0.53 0.44] If in Creative Royale use the default playlist tag since it doesn't use Athena.Playlist.Solo/Duo/Trio/Squad.
  (if (not (ScalableFloat|GetValueasBool (Variables|Creative|GetHotFix_UseDefaultPlaylistTagForCR)))
    ;[0.8353 1 0] Get the playlist tag to determine if it's solo/squad etc.
    ; Used in the degrading loot child class and was useful when giving bespoke rewards (legendary augments)
    (Utilities|IsValid _returnvalue
      (:Is%20Valid
        (for _array_element (Variables|Objective|Playlist|GetPossiblePlaylistTags)
          (if (GameplayTags|TagContainer|CompareTagContainertoOtherTags _array_element (Class|FortGameStateAthena|GetAthenaPlaylistContextTags _returnvalue))
            (return _array_element)))))
    (else
      (return "(TagName=\"Athena.Playlist.Duo\")"))))

(fn GetAllAssignedSquadIDs
  ;[0.4687 0.8063 1] GetAllAssignedSquadIDs
  ; Gets the stored assigned players and returns the squad IDs of all in an array
  (Utilities|Array|Clear (LocalVariables|GetReturnAllAssignedSquadIDs))
  (for _array_element (Variables|Objective|Assigned|GetPlayersAssignedToObjective)
    (Utilities|Array|AddUnique (LocalVariables|GetReturnAllAssignedSquadIDs) (Math|Conversions|ToInteger(Byte) (Team|GetActorTeam _array_element))))
  (return (LocalVariables|GetReturnAllAssignedSquadIDs)))

(fn GetAnalyticsAttributes
  (bind _ecurrentobjectivestate_repnotify (Variables|Objective|States|GetECurrentObjectiveState_RepNotify))
  (bind _returnvalue (Math|Byte|MakeLiteralByte _ecurrentobjectivestate_repnotify))
  (bind _playlistsquadtag (Reward|GetPlaylistSquadSizeTag))
  (bind _map_playlisttosquadsize (Variables|Objective|Playlist|GetMap_PlaylistToSquadSize))
  (bind _value (Utilities|Map|Find _map_playlisttosquadsize _playlistsquadtag))
  (bind _returnvalue_2 (> _returnvalue 2))
  (bind _analyticsattribute_activatingplayerassignedtoobjective (Variables|Analytics|GetAnalyticsAttribute_ActivatingPlayerAssignedToObjective))
  (bind _analyticsattribute_assignedascontested (Variables|Analytics|GetAnalyticsAttribute_AssignedAsContested))
  (bind _analyticsattribute_elapsedtimetoobjectiveactivation (Variables|Analytics|GetAnalyticsAttribute_ElapsedTimeToObjectiveActivation))
  ;[0.2435 0.1798 0.4062] Comment
  (Utilities|Array|Clear (LocalVariables|GetReturn_AnalyticsEventAttributes))
  (Utilities|FlowControl|Sequence
    (:then_0
      ;;[0.2435 0.1798 0.4062] Sequence [0]: Analytics attributes that should be sent regardless of interactivity - related to initiating the objective
      ;;;[0.2435 0.1798 0.4062] Attribute: Was this objective assigned to 2 or more squads?
      (bind _returnvalue_3 (Analytics|ConstructAnalyticsAttribute :AttributeName "AssignedAsContestedObjective" :IntValue -1 :floatValue -1.0 :BoolValue _analyticsattribute_assignedascontested))
      (Utilities|Array|Add (LocalVariables|GetReturn_AnalyticsEventAttributes) _returnvalue_3)
      (Fortnite|FortShippingLog self (Utilities|String|Append "ASSIGNED OBJECTIVE ANALYTICS:  AssignedAsContestedObjective = " (Utilities|String|ToString(Boolean) _analyticsattribute_assignedascontested)))
      ;;;[0.2435 0.1798 0.4062] Attribute: What is the squad size of this playlist? - gives context for other analytics
      (bind _returnvalue_4 (Analytics|ConstructAnalyticsAttribute :AttributeName "ContextPlaylistSquadSize" :IntValue _value :floatValue -1.0 :BoolValue false))
      (Utilities|Array|Add (LocalVariables|GetReturn_AnalyticsEventAttributes) _returnvalue_4)
      (Fortnite|FortShippingLog self (Utilities|String|Append "ASSIGNED OBJECTIVE ANALYTICS:  ContextPlaylistSquadSize = " (Utilities|String|ToString(Integer) _value)))
      ;;;[0.2435 0.1798 0.4062] Attribute: What is the distance for all assigned squads - based on the average midpoint of all the players (passed from the Game State Component)
      (for (_array_element _array_index) (Variables|Analytics|GetAnalyticsAttribute_AssignedTeamDistance)
        (bind _returnvalue_5 (Analytics|ConstructAnalyticsAttribute :AttributeName (Utilities|String|Append "DistanceOfSquadToObjective_squad" (Utilities|String|ToString(Integer) _array_index)) :IntValue -1 :floatValue _array_element :BoolValue false))
        (Utilities|Array|Add (LocalVariables|GetReturn_AnalyticsEventAttributes) _returnvalue_5)
        (Fortnite|FortShippingLog self (Utilities|String|Append "ASSIGNED OBJECTIVE ANALYTICS: " (Utilities|String|Append "DistanceOfSquadToObjective_squad" (Utilities|String|ToString(Integer) _array_index)) " = " (Utilities|String|ToString(Float) _array_element)))))
    (:then_1
      ;;[0.2435 0.1798 0.4062] Sequence [1]: Attributes that are related to the objective being activated. Filters out attributes if the objective was not activated
      ;;;[0.2435 0.1798 0.4062] Attribute: Was the objective activated?
      (bind _returnvalue_6 (Analytics|ConstructAnalyticsAttribute :AttributeName "WasObjectiveActivated" :IntValue -1 :floatValue -1.0 :BoolValue _returnvalue_2))
      (Utilities|Array|Add (LocalVariables|GetReturn_AnalyticsEventAttributes) _returnvalue_6)
      (Fortnite|FortShippingLog self (Utilities|String|Append "ASSIGNED OBJECTIVE ANALYTICS:  WasObjectiveActivated = " (Utilities|String|ToString(Boolean) _returnvalue_2)))
      (if _returnvalue_2
        ;;;[0.2435 0.1798 0.4062] Attribute: How long before the objective was activated since it spawned?
        (bind _returnvalue_7 (Analytics|ConstructAnalyticsAttribute :AttributeName "ElapsedTimeBeforeObjectiveActivation" :IntValue -1 :floatValue _analyticsattribute_elapsedtimetoobjectiveactivation :BoolValue false))
        (Utilities|Array|Add (LocalVariables|GetReturn_AnalyticsEventAttributes) _returnvalue_7)
        (Fortnite|FortShippingLog self (Utilities|String|Append "ASSIGNED OBJECTIVE ANALYTICS:  ElapsedTimeBeforeObjectiveActivation = " (Utilities|String|ToString(Float) _analyticsattribute_elapsedtimetoobjectiveactivation)))
        ;;;[0.2435 0.1798 0.4062] Attribute: Was the player who activated the objective assigned to it or are they "stealing" it?
        (bind _returnvalue_8 (Analytics|ConstructAnalyticsAttribute :AttributeName "ActivatingPlayerAssignedToObjective" :IntValue -1 :floatValue -1.0 :BoolValue _analyticsattribute_activatingplayerassignedtoobjective))
        (Utilities|Array|Add (LocalVariables|GetReturn_AnalyticsEventAttributes) _returnvalue_8)
        (Fortnite|FortShippingLog self (Utilities|String|Append "ASSIGNED OBJECTIVE ANALYTICS:  ActivatingPlayerAssignedToObjective = " (Utilities|String|ToString(Boolean) _analyticsattribute_activatingplayerassignedtoobjective)))))
    (:then_3
      ;;[0.9707 0.4876 1] RETURN: Send all the tracked analytic events (some are filtered based on the state of the objective.
      (return (LocalVariables|GetReturn_AnalyticsEventAttributes)))))

(fn GetCacheContextOverride
  ;[0.4163 1 0.9801] We don't need to return anything here since we don't need to override the Cache Context.
  (return 0))

(fn GetCurrentLootTier
  (bind (_value _returnvalue) (Utilities|Map|Find (Variables|Objective|Reward|GetLootGroupByPlaylistMap) (Reward|GetPlaylistSquadSizeTag)))
  (bind (_value_2 _returnvalue_2) (Utilities|Map|Find (Utilities|Select :Index (Utilities|Map|Find (Variables|Objective|Reward|GetMap%20from%20Float%20To%20Rarity%20Enum) (ScalableFloat|GetValueatLevel (Variables|Default|GetHF_LootRarity))) :Common (Variables|Objective|Reward|GetMap_RareLoot_PlaylistToLootTier) :Uncommon (Variables|Objective|Reward|GetMap_RareLoot_PlaylistToLootTier) :Rare (Variables|Objective|Reward|GetMap_RareLoot_PlaylistToLootTier) :Epic (Variables|Objective|Reward|GetMap_EpicLoot_PlaylistToLootTier) :Legendary (Variables|Objective|Reward|GetMap_LegendaryLoot_PlaylistToLootTier) :Mythic (Variables|Objective|Reward|GetMap_LegendaryLoot_PlaylistToLootTier) :Transcendent (Variables|Objective|Reward|GetMap_LegendaryLoot_PlaylistToLootTier) :Unattainable (Variables|Objective|Reward|GetMap_LegendaryLoot_PlaylistToLootTier)) (Reward|GetPlaylistSquadSizeTag)))
  ;[0 0.1068 0.6441] GetCurrentLootTier: Hotfix values for forcing rarity for Loot_MMO_ChestCache tiers, OR set loot tier to Loot_ClaimPOI_Reward_
  (if (ScalableFloat|GetValueasBool (Variables|Default|GetHF_LootTierType))
    (return _value_2 _returnvalue_2)
    (else
      (return _value _returnvalue))))

(fn HandleGamePhaseChanged
  (params (GamePhase :type EAthenaGamePhase%20Enum))
  ;[0.03332 0.3212 0.2859] HandleGamePhaseChanged
  ; When the game is ended, send the assigned objective analytics event for the match
  (if (Utilities|Enum|Equal(Enum) GamePhase "EndGame")
    ;;[0.2431 0.1804 0.4078] Don't fire analytics in Creative Royale.
    (if (not (ScalableFloat|GetValueasBool (Variables|Creative|GetHotFix_UseDefaultPlaylistTagForCR)))
      ;;[0.2435 0.1798 0.4062] Get and Fire all Analytic Attributes for the match (each objective will do this)
      (bind _analyticseventattributes (Analytics|GetAnalyticsAttributes))
      (Analytics|FireAnalyticsEventBlueprintArray (Variables|Analytics|GetAnalyticsEventName) _analyticseventattributes)
      (else
        _))
    (bind _gamephaseupdatedasynctask (Variables|Default|GetGamePhaseUpdatedAsyncTask)
      (:then
        (GameplayEventRouter|StopListeningtoStatefulEvent _gamephaseupdatedasynctask))
      (:else))))

(fn HasPlayersAssignedToObjectiveReplicatedToThisClient
  ;[0.3202 0.7376 1] HasPlayersAssignedToObjectiveReplicatedToThisClient
  ; Get if the asssigned players is replicated - since the objective is highly reliant on assigned/unassigned for communication
  (return (and (not (Networking|IsDedicatedServer)) (Utilities|Array|IsNotEmpty (Variables|Objective|Assigned|GetPlayersAssignedToObjective)))))

(fn IsLocalClientAssignedToObjective
  ;[0.3202 0.7376 1] IsLocalClientAssignedToObjective [PURE]
  ; Checks local player controllers against the current assigned players array
  ; Use the view target's player state instead of the controller's playerstate so that this check works for spectators
  (Utilities|FlowControl|Sequence
    (:then_0
      (for _array_element (Fortnite|GetLocalFortPlayerControllers true true)
        (Utilities|IsValid _array_element
          (:Is%20Valid
            (bind _aspawn (Utilities|Casting|CastToPawn (Pawn|GetViewTarget _array_element)))
            (bind _asfort_player_state (Utilities|Casting|CastToFortPlayerState (Variables|/Script/Engine.Pawn|GetPlayerState _aspawn)))
            (if (Utilities|Array|ContainsItem (Variables|Objective|Assigned|GetPlayersAssignedToObjective) _asfort_player_state)
              ;;[1 0.3594 0.9217] Return true if (any) local player is assigned
              (return true))))))
    (:then_1
      ;;[1 0.3594 0.9217] Return false in all other cases
      (return false))))

(fn IsPawnAssignedToObjective
  (params (PawnToCheck :type Fort%20Pawn%20Object%20Reference))
  ;[0.3202 0.7376 1] IsPawnAssignedToObjective
  ; Check if player is assigned to the objective
  (Utilities|FlowControl|Sequence
    (:then_0
      ;;[0.3202 0.7376 1] [0] Sequence is valid
      ;; Check pawn against the players assigned to objective array and return the result
      (Utilities|IsValid PawnToCheck
        (:Is%20Valid
          (bind _asfort_player_state (Utilities|Casting|CastToFortPlayerState (Variables|/Script/Engine.Pawn|GetPlayerState PawnToCheck)))
          ;;;[1 0.3594 0.9217] Return: player in assigned players llist
          (return (Utilities|Array|ContainsItem (Variables|Objective|Assigned|GetPlayersAssignedToObjective) _asfort_player_state)))))
    (:then_1
      ;;[1 0.3594 0.9217] [1] Sequence catches all invalid nodes
      ;; Return: false due to null check
      (return false))))

(fn OnRep_bObjectiveSpawnedAndReady
  (bind _bobjectivespawnedandready (Variables|Objective|States|GetbObjectiveSpawnedAndReady))
  ;[0.2135 0.09237 0] Update the Objective "Spawned And Ready": enable or disable the mesh, disable interaction, and collision
  ; Needs to run on Server and Clients
  ; Is not included in the rest of the state tracking since this logic needs to persist on all following states, is not state-specific after Spawning
  (Variables|Objective|Interaction|SetbCanInteract _bobjectivespawnedandready)
  (Variables|Interaction|SetbShowFirstInteractPrompt _bobjectivespawnedandready))

(fn OnRep_ECurrentObjectiveState_RepNotify
  ;[0.2687 0.7669 1] OnRep
  ; ECurrentObjectiveState updated
  ; Replicated to allow for cosmetic events - enum ensures only one valid state is active at a time
  (CallFunction|ObjectiveStateChanged) ; must be an event to allow a delay node
)

(fn OnRep_NitroDromeGameplayVolume
  (CallFunction|UpdateMMOCrateMaterialCaptureState 0.0)
  (Rendering|SetVisibility (Variables|Default|GetInteractionObject) true false))

(fn OnRep_RemovePrompts)

(fn SendTutorialQuestEventToAssignedPlayers
  ;[0 0.2656 1] SendTutorialEventToAssignedPlayers
  ; Each assigned player should get a quest update that the objective is spawned
  (for _array_element (Variables|Objective|Assigned|GetPlayersAssignedToObjective)
    (Utilities|IsValid (Fortnite|GetFortPlayerControllerfromActor _array_element)
      (:Is%20Valid
        (Utilities|IsValid (Quests|GetQuestManager (Fortnite|GetFortPlayerControllerfromActor _array_element) "Invalid")
          (:Is%20Valid
            (Quest|SendComplexCustomStatEvent (Quests|GetQuestManager (Fortnite|GetFortPlayerControllerfromActor _array_element) "Invalid") 0 (Variables|Objective|Quests|GetTutorialQuestTag) (GameplayTags|GetOwnedGameplayTags self :!class /Script/GameplayTags.GameplayTagAssetInterface) 1)
            (Ability|SendGameplayEventtoActor (Fortnite|GetFortPlayerControllerfromActor _array_element) "(TagName=\"Athena.Quest.MMObj.LootCache.Activate\")")))))))

(fn SetMMOMiniMapIcon
  (params (Captured :type Boolean))
  (if Captured
    (MiniMapData|SetMiniMapIcon (Variables|Default|GetFortMiniMapComponent) (Variables|Default|GetMini%20Map%20Icon%20Captured))
    (else
      (MiniMapData|SetMiniMapIcon (Variables|Default|GetFortMiniMapComponent) (Variables|Default|GetMini%20Map%20Icon%20Uncaptured)))))

(fn StartTakeOverGCNL
  (bind _playerstatethatstartedobjective (Variables|Objective|Interaction|GetPlayerStateThatStartedObjective)
    (:then
      (LocalVariables|SetEffectCauser (PlayerState|GetPawn _playerstatethatstartedobjective)))
    (:else
      _)))

(fn SwitchOnStateChange_DetermineCosmeticPersist
  ;[0.03332 0.3212 0.2859] StateChangeDetermineCosmeticPersist
  ; Get the persistent cosmetic event that will be called on state changes and relevancy changes
  (switch Utilities|FlowControl|Switch|SwitchonEAssignedInteractObjectiveState (Variables|Objective|States|GetECurrentObjectiveState_RepNotify)
    (:NewEnumerator1
      (CallFunction|CosmeticsStateSpawningPersistentLoop))
    (:NewEnumerator2
      (CallFunction|CosmeticsStateAvailableIdlePersistentLoop))
    (:NewEnumerator3
      (CallFunction|CosmeticsStateActiveObjectivePersistentLoop))
    (:NewEnumerator4
      (CallFunction|CosmeticsStateCompletedIdlePersistentLoop))
    (:NewEnumerator5
      ;;[1 0 0.9539] Unexpected "COUNT" enum state
      (Fortnite|FortShippingLog 0 "ASSIGNED MIDMATCH OBJECTIVE: Unexpected attempt to use \"COUNT\" enum state is not supported - persist check" true))))

(fn SwitchOnStateChange_DetermineStart
  ;[0.03332 0.3212 0.2859] StateChangeDetermineStart
  ; Determine the state events to call based on the Objective State Enum
  (switch Utilities|FlowControl|Switch|SwitchonEAssignedInteractObjectiveState (Variables|Objective|States|GetECurrentObjectiveState_RepNotify)
    (:NewEnumerator1
      (CallFunction|ObjectiveStateSpawning)
      ;;[0.3658 0 1] Cosmetic events for VFX/SFX etc.
      (Networking|CanExecuteCosmeticEvents
        (:True
          (CallFunction|CosmeticsStateSpawningStartBurst))))
    (:NewEnumerator2
      (CallFunction|ObjectiveStateSpawningEnd)
      (CallFunction|ObjectiveStateAvailableIdle)
      (Networking|CanExecuteCosmeticEvents
        (:True
          (CallFunction|CosmeticsStateSpawningEndBurst)
          (CallFunction|CosmeticsStateAvailableIdleStartBurst))))
    (:NewEnumerator3
      (CallFunction|ObjectiveStateActiveObjective)
      (Networking|CanExecuteCosmeticEvents
        (:True
          (CallFunction|CosmeticsStateActiveObjectiveStartBurst))))
    (:NewEnumerator4
      (CallFunction|ObjectiveStateCompletedIdle)
      (Networking|CanExecuteCosmeticEvents
        (:True
          (CallFunction|CosmeticsStateCompletedIdleStartBurst))))
    (:NewEnumerator5
      ;;[1 0 0.9539] Unexpected "COUNT" enum state
      (Fortnite|FortShippingLog 0 "ASSIGNED MIDMATCH OBJECTIVE: Unexpected attempt to use \"COUNT\" enum state is not supported - start check" true))))

(fn TossPickupsToEnsureLootSticksToTrain
  (params (Item%20Pickup%20Array :type Array%20of%20Fort%20Pickup%20Object%20References))
  (bind (_returnvalue_roll _returnvalue_pitch _returnvalue_yaw) (Math|Rotator|BreakRotator (Math|Rotator|MakeRotfromX (Transformation|GetActorForwardVector))))
  (for (_array_element _array_index) Item%20Pickup%20Array
    (Fortnite|Container|TossPickupfromContainer self _array_element (Utilities|Array|Length Item%20Pickup%20Array) _array_index 2.7e+02 (Math|Rotator|MakeRotator 0.0 15.0 _returnvalue_yaw) 5e+02)))

(fn ConstructionScript)
