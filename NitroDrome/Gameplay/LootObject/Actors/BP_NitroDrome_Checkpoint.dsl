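; NitroDrome checkpoint that players hit in vehicles to shorten the loot objective timer.
; On the authority, a vehicle overlap broadcasts Checkpoint.Used and plays cue and FX, then the checkpoint hides and relocates after a delay using the CheckpointManager's transform array, optionally reshuffled on wrap.
; RepNotify variables replicate transform and visibility/overlaps.
; Router events disable it on loot reward spawn and re-enable it on reset; hotfix ScalableFloats can destroy it at BeginPlay.
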(event Custom|ManageCheckpointsBasedOnRewardState
  ;[1 1 1] Singleton async to wait for reward spawned and disable checkpoints. Start watching for loot event reset to re-enable
  (Networking|SwitchHasAuthority
    (:Authority
      (bind _returnvalue (GameplayEventRouter|GetEventRouter "/NitroDrome/Gameplay/Actors/BP_NitroDromePlayspace.BP_NitroDromePlayspace_C" self))
      (bind (_asynctaskproxy _payload) (GameplayEventRouter|StartListeningToEvent _returnvalue "/Script/NitroDromeRuntime.OnNitroDromeEventTriggered")
        (:OnEventReceived
          (bind _payload_eventtag (Utilities|Struct|BreakOnNitroDromeEventTriggered _payload))
          (switch Utilities|FlowControl|Switch|SwitchonGameplayTag _payload_eventtag
            (:NitroDrome.Event.LootRewardSpawned
              (Variables|Default|SetCheckpointDisabled true)
              (Variables|Default|SetRepNotifySetVisibilityAndOverlaps false))
            (:NitroDrome.Event.LootEventReset
              (Variables|Default|SetCheckpointDisabled false)
              (Variables|Default|SetRepNotifySetVisibilityAndOverlaps true))))))))

(event Collision|EventActorBeginOverlap
  (params (OtherActor :type Actor%20Object%20Reference)))

(event EventTick
  (params (DeltaSeconds :type Float%20%28single-precision%29)))

(event EventBeginPlay
  ;[0.005155 0 1] Check hotfix values for enabled. IF objective is disabled, also disable this
  (Class|FortAsyncActionCurrentPlaylistReady|PlaylistDataReadyAsync
    (:OnReady
      (if (and (ScalableFloat|GetValueasBool (Variables|ScalableFloats|GetSF_ObjectiveEnabled)) (ScalableFloat|GetValueasBool (Variables|ScalableFloats|GetSF_Enabled)))
        (Variables|Default|SetCooldownBeforeNewCP (ScalableFloat|GetValueatLevel (Variables|ScalableFloats|GetDelayBeforeNewCPAppears)))
        (CallFunction|ManageCheckpointsBasedonRewardState)
        (else
          (Actor|DestroyActor))))))

(event OnComponentBeginOverlap(CollisionExtents)
  (params (OverlappedComponent :type Primitive%20Component%20Object%20Reference) (OtherActor :type Actor%20Object%20Reference) (OtherComp :type Primitive%20Component%20Object%20Reference) (OtherBodyIndex :type Integer) (bFromSweep :type Boolean) (SweepResult :type Hit%20Result%20Structure%20%28by%20ref%29))
  (bind _checkpointmanager (Variables|Default|GetCheckpointManager))
  (bind _checkpointmanager_2 (Variables|Default|GetCheckpointManager))
  (bind _checkpointtransformarray (Variables|/NitroDrome/Gameplay/LootObject/Actors/NitroDrome_CheckpointManager.NitroDrome_CheckpointManager_C|GetCheckpointTransformArray _checkpointmanager))
  (bind _arrayindex (Variables|/NitroDrome/Gameplay/LootObject/Actors/NitroDrome_CheckpointManager.NitroDrome_CheckpointManager_C|GetArrayIndex _checkpointmanager_2))
  ;[0 0.9112 1] [SERVER]  checkpoints are set to lower objective timer to make loot spawn faster
  (Networking|SwitchHasAuthority
    (:Authority
      (bind _asfort_player_pawn_athena (Utilities|Casting|CastToFortPlayerPawnAthena OtherActor))
      (if (Vehicle|IsinVehicle _asfort_player_pawn_athena)
        (CallFunction|SendAffectEvent)
        (GameplayCue|ExecuteGameplayCue _asfort_player_pawn_athena (Variables|Default|GetCP%20Hit%20Gameplay%20Cue%20Tag))
        ;[0 0.2085 1] Checkpoint was hit; Hide the BGA, move to new transform after delay, reset checkpoint array if we've run through all locations already. Unhide
        (CallFunction|PlayPassFX)
        (Variables|Default|SetRepNotifySetVisibilityAndOverlaps false)
        (Utilities|FlowControl|Delay (Variables|Default|GetCooldownBeforeNewCP))
        (Variables|Default|SetCheckpointTransform (Utilities|Array|Get(acopy) (Variables|/NitroDrome/Gameplay/LootObject/Actors/NitroDrome_CheckpointManager.NitroDrome_CheckpointManager_C|GetCheckpointTransformArray _checkpointmanager_2) _arrayindex))
        (Math|Integer|IncrementInt _arrayindex
          (:%20%20
            (Utilities|FlowControl|Delay (Variables|Default|GetTimeAfterMove))
            (if (>= (Variables|/NitroDrome/Gameplay/LootObject/Actors/NitroDrome_CheckpointManager.NitroDrome_CheckpointManager_C|GetArrayIndex _checkpointmanager) (Utilities|Array|Length _checkpointtransformarray))
              (Variables|/NitroDrome/Gameplay/LootObject/Actors/NitroDrome_CheckpointManager.NitroDrome_CheckpointManager_C|SetArrayIndex 0 _checkpointmanager)
              (if (ScalableFloat|GetValueasBool (Variables|ScalableFloats|GetSF_RandomOrder))
                (Utilities|Array|Shuffle _checkpointtransformarray)
                (else
                  _)))
            (if (not (Variables|Default|GetCheckpointDisabled))
              (Variables|Default|SetRepNotifySetVisibilityAndOverlaps true)
              (CallFunction|PlayPassFX))))))))

(event Custom|SendAffectEvent
  ;[0 0.9112 1] [SERVER]  checkpoints are set to lower objective timer to make loot spawn faster
  (bind _returnvalue (GameplayEventRouter|GetEventRouter "/NitroDrome/Gameplay/Actors/BP_NitroDromePlayspace.BP_NitroDromePlayspace_C" self))
  (GameplayEventRouter|BroadcastEvent _returnvalue (Utilities|Struct|MakeOnNitroDromeReactableEventHappened "(GameplayTags=((TagName=\"NitroDrome.Event.Checkpoint.Used\")))") 0))

(event Custom|PlayPassFX
  (Components|Activation|Activate (Variables|Default|GetNS_CheckPoint_Pass_Blueprint) false))

(fn OnRep_CheckpointTransform
  (Transformation|SetActorTransform (Variables|Default|GetCheckpointTransform) false true))

(fn OnRep_RepNotifySetVisibilityAndOverlaps
  (Variables|/Script/Engine.PrimitiveComponent|SetbGenerateOverlapEvents (Variables|Default|GetRepNotifySetVisibilityAndOverlaps) (Variables|Default|GetCollisionExtents))
  (Rendering|SetVisibility (Variables|Default|GetStaticMesh) (Variables|Default|GetRepNotifySetVisibilityAndOverlaps) false))

(fn ConstructionScript)
