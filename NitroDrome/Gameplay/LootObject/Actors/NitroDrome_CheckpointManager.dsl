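; NitroDrome checkpoint manager actor that spawns the first checkpoint of the objective from an array of transforms.
; On BeginPlay it waits for playlist data, then checks the scalable float enable flags and destroys itself if either is off.
; It optionally shuffles the transforms, spawns a checkpoint at the current index, increments the index and links the checkpoint back to itself.
; An empty array only logs; later checkpoint advancement is not in this graph.
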(event Collision|EventActorBeginOverlap
  (params (OtherActor :type Actor%20Object%20Reference)))

(event EventTick
  (params (DeltaSeconds :type Float%20%28single-precision%29)))

(event EventBeginPlay
  (bind _checkpointtransformarray (Variables|Default|GetCheckpointTransformArray))
  ;[0.3746 0 1] Check hotfix values for enabled & if only 1 station will reset vs multiple
  (Class|FortAsyncActionCurrentPlaylistReady|PlaylistDataReadyAsync
    (:OnReady
      (if (and (ScalableFloat|GetValueasBool (Variables|Default|GetSF_ObjectiveEnabled)) (ScalableFloat|GetValueasBool (Variables|Default|GetSF_Enabled)))
        (if (> (Utilities|Array|Length _checkpointtransformarray) 0)
          (if (ScalableFloat|GetValueasBool (Variables|Default|GetSF_RandomOrder))
            (Utilities|Array|Shuffle _checkpointtransformarray)
            (else
              _))
          (bind _returnvalue (Building|SpawnBuildingGameplayActor "/NitroDrome/Gameplay/LootObject/Actors/BP_NitroDrome_Checkpoint.BP_NitroDrome_Checkpoint_C" (Utilities|Array|Get(acopy) (Variables|Default|GetCheckpointTransformArray) (Variables|Default|GetArrayIndex)) self true))
          (Math|Integer|IncrementInt (Variables|Default|GetArrayIndex)
            (:%20%20
              (bind _asbp_nitro_drome_checkpoint (Utilities|Casting|CastToBP_NitroDrome_Checkpoint _returnvalue))
              (Variables|/NitroDrome/Gameplay/LootObject/Actors/BP_NitroDrome_Checkpoint.BP_NitroDrome_Checkpoint_C|SetCheckpointManager self _asbp_nitro_drome_checkpoint)
              (Variables|Default|SetCheckPointBGA _asbp_nitro_drome_checkpoint)))
          (else
            (Utilities|String|LogString "Checkpoint Array Empty")))
        (else
          (Actor|DestroyActor))))))

(fn ConstructionScript)
